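# Physics asset for the female medium Division Jacket body cosmetic, set up for its dangling back dress and front flap bones.
# Capsules on the pelvis and both thighs form the body volumes, joined by free angular constraints.
# Two box bodies form a back dress chain from the pelvis with limited swing and twist.
# Three front flap capsules chain from a small master body free on the pelvis, with one swing axis much looser than the other.
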PhysicsAsset /BRCosmetics/Characters/Player/Female/Medium/Bodies/F_MED_Division_Jacket/Meshes/F_MED_Division_Physics
bodies {
 dyn_backDress_01 [{shapeType=Box, extentX=8.8207, extentY=25.3992, extentZ=3.0816, shapeName=dyn_backDress_01_box, center=<4.7006 -0.0001 0>}]
 dyn_backDress_02 [{shapeType=Box, extentX=8.8207, extentY=25.3992, extentZ=3.0816, shapeName=dyn_backDress_02_box, center=<4.7006 -0.0001 0>}]
 dyn_frontFlap_01 [{shapeType=Capsule, length=28.7016, radius=2.3472, shapeName=dyn_frontFlap_01_capsule, center=<3.8872 0 1.2011>, rotation=<rot 0 0 90>}]
 dyn_frontFlap_02 [{shapeType=Capsule, length=22.2053, radius=2.3314, shapeName=dyn_frontFlap_02_capsule, center=<3.8872 0 1.1888>, rotation=<rot 0 0 90>}]
 dyn_frontFlap_03 [{shapeType=Capsule, length=13.0153, radius=2.6077, shapeName=dyn_frontFlap_03_capsule, center=<2.7945 0 1.0873>, rotation=<rot 0 0 90>}]
 dyn_frontFlap_MSTR [{shapeType=Capsule, length=0.505, radius=0.505, shapeName=dyn_frontFlap_MSTR_capsule}]
 pelvis [{shapeType=Capsule, length=13.2443, radius=8.0725, shapeName=pelvis_capsule}]
 thigh_l [{shapeType=Capsule, length=30.8132, radius=6.7444, shapeName=thigh_r_capsule, center=<-20.3161 -0.0005 0>, rotation=<rot 90 0 0>}]
 thigh_r [{shapeType=Capsule, length=30.8132, radius=6.7444, shapeName=thigh_r_capsule, center=<20.3603 -0.0005 0>, rotation=<rot 90 0 0>}]
}
constraints [
 {
  bone1Name=thigh_l
  bone2Name=pelvis
  swing1LimitDegrees=45
  swing1Motion=Free
  swing2LimitDegrees=45
  swing2Motion=Free
  twistLimitDegrees=45
  twistMotion=Free
 }
 {
  bone1Name=thigh_r
  bone2Name=pelvis
  swing1LimitDegrees=45
  swing1Motion=Free
  swing2LimitDegrees=45
  swing2Motion=Free
  twistLimitDegrees=45
  twistMotion=Free
 }
 {bone1Name=dyn_backDress_01, bone2Name=pelvis, swing1LimitDegrees=30, swing2LimitDegrees=30, twistLimitDegrees=30}
 {bone1Name=dyn_backDress_02, bone2Name=dyn_backDress_01, swing1LimitDegrees=30, swing2LimitDegrees=30, twistLimitDegrees=30}
 {
  bone1Name=dyn_frontFlap_MSTR
  bone2Name=pelvis
  swing1LimitDegrees=45
  swing1Motion=Free
  swing2LimitDegrees=45
  swing2Motion=Free
  twistLimitDegrees=45
  twistMotion=Free
 }
 {bone1Name=dyn_frontFlap_01, bone2Name=dyn_frontFlap_MSTR, swing1LimitDegrees=15, swing2LimitDegrees=60, twistLimitDegrees=15}
 {bone1Name=dyn_frontFlap_02, bone2Name=dyn_frontFlap_01, swing1LimitDegrees=15, swing2LimitDegrees=60, twistLimitDegrees=15}
 {bone1Name=dyn_frontFlap_03, bone2Name=dyn_frontFlap_02, swing1LimitDegrees=15, swing2LimitDegrees=60, twistLimitDegrees=15}
]
