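# Physics asset for the Division jacket cosmetic's crouching variant, setting up collision and joints for the jacket's dynamic bones.
# Front, side and back panels on each side are four-bone chains with angular limits, hung from tiny master bodies linked freely to the pelvis.
# Pelvis, thigh and calf shapes form the body, joined by free constraints up to root, which carries a very large offset sphere.
# Shape sizes and limits differ noticeably between left and right.
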PhysicsAsset /BRCosmetics/Characters/Player/Female/Medium/Bodies/F_MED_Division_Jacket/Meshes/M_MED_Division_Jacket_Crouching_Physics
bodies {
 calf_l [{shapeType=Capsule, length=50.0318, radius=7.5323, center=<-23.8904 -1.0428 -0.0001>, rotation=<rot 90 -18.9704 -24.9704>}]
 calf_r [
  {shapeType=Capsule, length=55.5884, radius=11.1378, center=<30 2 -1.0863>, rotation=<rot 90 -90 -92.6013>}
  {shapeType=Capsule, length=60.4071, radius=6.3191, shapeName=calf_r_capsule, center=<31.3491 -12.1281 -2.6722>, rotation=<rot 90 0 0>}
 ]
 dyn_jacketBack_01_l [{shapeType=Capsule, length=7.8918, radius=2.1082, shapeName=dyn_jacketBack_01_l_capsule, center=<6.247 -0.0002 0>, rotation=<rot 0 0 90>}]
 dyn_jacketBack_01_r [{shapeType=Capsule, length=7.7086, radius=2.2914, shapeName=dyn_jacketBack_01_r_capsule, center=<6.247 -0.0002 0>, rotation=<rot 0 0 90>}]
 dyn_jacketBack_02_l [{shapeType=Capsule, length=9.4554, radius=5.4547, shapeName=dyn_jacketBack_02_l_capsule, center=<6.9678 0.0001 0.053>, rotation=<rot 90 0 0>}]
 dyn_jacketBack_02_r [{
   shapeType=Capsule
   length=52.7395
   radius=7.2511
   shapeName=dyn_jacketBack_02_r_capsule
   center=<11.4886 0.0001 0.053>
   rotation=<rot 90 -45 -132.8625>
  }]
 dyn_jacketBack_03_l [{shapeType=Capsule, length=25.2417, radius=5.6226, shapeName=dyn_jacketBack_03_l_capsule, center=<9.1877 0 0.0001>, rotation=<rot 0 0 90>}]
 dyn_jacketBack_03_r [{shapeType=Capsule, length=51.577, radius=12.2761, shapeName=dyn_jacketBack_03_r_capsule, center=<9.0544 0 0>, rotation=<rot 0 0 90>}]
 dyn_jacketBack_04_l [{shapeType=Capsule, length=28.2412, radius=4.088, shapeName=dyn_jacketBack_04_l_capsule, center=<6.247 -0.0002 0>, rotation=<rot 0 0 90>}]
 dyn_jacketBack_04_r [{shapeType=Capsule, length=52.2653, radius=12.5133, shapeName=dyn_jacketBack_04_r_capsule, center=<6.247 -0.0002 0>, rotation=<rot 0 0 90>}]
 dyn_jacketBack_mstr_l [{shapeType=Capsule, length=0.505, radius=0.505, rotation=<rot 0 -90 0>}]
 dyn_jacketBack_mstr_r [{shapeType=Capsule, length=0.505, radius=0.505, rotation=<rot 0 -90 0>}]
 dyn_jacketFront_01_l [{shapeType=Capsule, length=15.9816, radius=3.8476, shapeName=dyn_jacketFront_01_r_capsule, center=<7.2591 0 0>, rotation=<rot 0 0 -75.6229>}]
 dyn_jacketFront_01_r [{
   shapeType=Box
   extentX=16.5887
   extentY=15.1282
   extentZ=1.9949
   shapeName=dyn_jacketFront_01_r_box
   center=<8.2197 0 0>
   rotation=<rot 0 0 -39.1605>
  }]
 dyn_jacketFront_02_l [{
   shapeType=Capsule
   length=11.655
   radius=3.3283
   shapeName=dyn_jacketFront_02_l_capsule
   center=<8.2332 0.753 -0.0278>
   rotation=<rot 84.0657 0.0001 0.0001>
  }]
 dyn_jacketFront_02_r [{
   shapeType=Capsule
   length=12.5394
   radius=2.4439
   shapeName=dyn_jacketFront_02_r_capsule
   center=<6.247 -0.0649 0.1399>
   rotation=<rot 0 0 45.3843>
  }]
 dyn_jacketFront_03_l [{
   shapeType=Capsule
   length=9.9784
   radius=5.0049
   shapeName=dyn_jacketFront_03_l_capsule
   center=<6.247 -0.0001 0.2754>
   rotation=<rot 0 0 130.3902>
  }]
 dyn_jacketFront_03_r [{
   shapeType=Capsule
   length=16.8013
   radius=2.1411
   shapeName=dyn_jacketFront_03_r_capsule
   center=<8.2668 -0.0002 0>
   rotation=<rot 0 0 49.9708>
  }]
 dyn_jacketFront_04_l [{
   shapeType=Capsule
   length=9.3447
   radius=5.6387
   shapeName=dyn_jacketFront_04_l_capsule
   center=<6.247 0 0.0798>
   rotation=<rot 0 0 127.3632>
  }]
 dyn_jacketFront_04_r [{
   shapeType=Capsule
   length=13.0488
   radius=1.9345
   shapeName=dyn_jacketFront_04_r_capsule
   center=<6.247 -0.0002 0>
   rotation=<rot 0 0 48.3905>
  }]
 dyn_jacketFront_mstr_l [{shapeType=Capsule, length=0.505, radius=0.505, rotation=<rot 0 -90 0>}]
 dyn_jacketFront_mstr_r [{shapeType=Capsule, length=0.505, radius=0.505, rotation=<rot 0 -90 0>}]
 dyn_jacketSide_01_l [{
   shapeType=Capsule
   length=11.1485
   radius=1.9117
   shapeName=dyn_jacketSide_01_l_capsule
   center=<6.2471 -0.0003 0.0912>
   rotation=<rot 90 -56.3099 -148.2146>
  }]
 dyn_jacketSide_01_r [{
   shapeType=Capsule
   length=6.304
   radius=4.1857
   shapeName=dyn_jacketSide_01_r_capsule
   center=<8.4049 -0.1193 3.2838>
   rotation=<rot 90 -172.875 100.2797>
  }]
 dyn_jacketSide_02_l [{shapeType=Capsule, length=13.224, radius=1.7593, shapeName=dyn_jacketSide_02_l_capsule, center=<6.247 0.0001 0.053>, rotation=<rot 0 0 90>}]
 dyn_jacketSide_02_r [{shapeType=Capsule, length=22.7713, radius=4.1389, shapeName=dyn_jacketSide_02_r_capsule, center=<8.2353 0.0001 3.116>, rotation=<rot 0 0 90>}]
 dyn_jacketSide_03_l [{shapeType=Capsule, length=10.9851, radius=3.9982, shapeName=dyn_jacketSide_03_l_capsule, center=<6.247 -0.0002 0>, rotation=<rot 0 0 90>}]
 dyn_jacketSide_03_r [{shapeType=Capsule, length=23.0649, radius=3.8452, shapeName=dyn_jacketSide_03_r_capsule, center=<6.6728 0.0002 0>, rotation=<rot 0 0 90>}]
 dyn_jacketSide_04_l [{shapeType=Capsule, length=13.2469, radius=3.9338, shapeName=dyn_jacketSide_04_l_capsule, center=<6.247 -0.0002 0>, rotation=<rot 0 0 90>}]
 dyn_jacketSide_04_r [{shapeType=Capsule, length=23.8584, radius=3.0517, shapeName=dyn_jacketSide_04_r_capsule, center=<3.8377 0.0002 0>, rotation=<rot 0 0 90>}]
 dyn_jacketSide_mstr_l [{shapeType=Capsule, length=0.505, radius=0.505, rotation=<rot 0 -90 0>}]
 dyn_jacketSide_mstr_r [{shapeType=Capsule, length=0.505, radius=0.505, rotation=<rot 0 -90 0>}]
 pelvis [
  {shapeType=Capsule, length=12.7328, radius=10.5564, center=<-0.649 0.2801 -0.0002>, rotation=<rot 90 0 -3.3073>}
  {
   shapeType=Capsule
   length=15.3785
   radius=7.9107
   shapeName=pelvis_capsule
   center=<-0.0808 -1.3903 -4.5558>
   rotation=<rot 80.001 -176.6924 -180>
  }
  {shapeType=Capsule, length=14.0875, radius=9.2018, shapeName=pelvis_capsule, center=<-0.2328 -1.3991 5.3434>, rotation=<rot 77.8519 3.3074 0>}
 ]
 root [
  {shapeType=Sphere, radius=183.4375, shapeName=root_sphere, center=<0 -8.6403 -175.5073>}
  {shapeType=Sphere, radius=27.8343, shapeName=root_sphere, center=<0.0002 2.4787 8.5972>}
 ]
 thigh_l [{shapeType=Capsule, length=29.0742, radius=6.9092, center=<-21.4337 0 -0.2548>, rotation=<rot 90 0 0>}]
 thigh_r [{shapeType=Capsule, length=40.8166, radius=7.156, center=<21.7868 0 0>, rotation=<rot 85.4571 -180 -180>}]
}
constraints [
 {
  bone1Name=thigh_l
  bone2Name=pelvis
  swing1LimitDegrees=45
  swing1Motion=Free
  swing2LimitDegrees=45
  swing2Motion=Free
  twistLimitDegrees=45
  twistMotion=Free
 }
 {
  bone1Name=calf_l
  bone2Name=thigh_l
  swing1LimitDegrees=45
  swing1Motion=Free
  swing2LimitDegrees=45
  swing2Motion=Free
  twistLimitDegrees=45
  twistMotion=Free
 }
 {
  bone1Name=thigh_r
  bone2Name=pelvis
  swing1LimitDegrees=45
  swing1Motion=Free
  swing2LimitDegrees=45
  swing2Motion=Free
  twistLimitDegrees=45
  twistMotion=Free
 }
 {
  bone1Name=calf_r
  bone2Name=thigh_r
  swing1LimitDegrees=45
  swing1Motion=Free
  swing2LimitDegrees=45
  swing2Motion=Free
  twistLimitDegrees=45
  twistMotion=Free
 }
 {
  bone1Name=dyn_jacketSide_mstr_l
  bone2Name=pelvis
  swing1LimitDegrees=45
  swing1Motion=Free
  swing2LimitDegrees=45
  swing2Motion=Free
  twistLimitDegrees=45
  twistMotion=Free
 }
 {
  bone1Name=dyn_jacketBack_mstr_l
  bone2Name=pelvis
  swing1LimitDegrees=45
  swing1Motion=Free
  swing2LimitDegrees=45
  swing2Motion=Free
  twistLimitDegrees=45
  twistMotion=Free
 }
 {
  bone1Name=dyn_jacketBack_mstr_r
  bone2Name=pelvis
  swing1LimitDegrees=45
  swing1Motion=Free
  swing2LimitDegrees=45
  swing2Motion=Free
  twistLimitDegrees=45
  twistMotion=Free
 }
 {
  bone1Name=dyn_jacketSide_mstr_r
  bone2Name=pelvis
  swing1LimitDegrees=45
  swing1Motion=Free
  swing2LimitDegrees=45
  swing2Motion=Free
  twistLimitDegrees=45
  twistMotion=Free
 }
 {
  bone1Name=dyn_jacketFront_mstr_l
  bone2Name=pelvis
  swing1LimitDegrees=45
  swing1Motion=Free
  swing2LimitDegrees=45
  swing2Motion=Free
  twistLimitDegrees=45
  twistMotion=Free
 }
 {
  bone1Name=dyn_jacketFront_mstr_r
  bone2Name=pelvis
  swing1LimitDegrees=45
  swing1Motion=Free
  swing2LimitDegrees=45
  swing2Motion=Free
  twistLimitDegrees=45
  twistMotion=Free
 }
 {bone1Name=dyn_jacketFront_01_r, bone2Name=dyn_jacketFront_mstr_r, swing1LimitDegrees=5, swing2LimitDegrees=45, twistLimitDegrees=5}
 {bone1Name=dyn_jacketFront_02_r, bone2Name=dyn_jacketFront_01_r, swing1LimitDegrees=5, swing2LimitDegrees=45, twistLimitDegrees=5}
 {bone1Name=dyn_jacketFront_03_r, bone2Name=dyn_jacketFront_02_r, swing1LimitDegrees=5, swing2LimitDegrees=45, twistLimitDegrees=5}
 {bone1Name=dyn_jacketFront_04_r, bone2Name=dyn_jacketFront_03_r, swing1LimitDegrees=5, swing2LimitDegrees=45, twistLimitDegrees=5}
 {bone1Name=dyn_jacketFront_01_l, bone2Name=dyn_jacketFront_mstr_l, swing1LimitDegrees=45, swing2LimitDegrees=45, twistLimitDegrees=25}
 {bone1Name=dyn_jacketFront_02_l, bone2Name=dyn_jacketFront_01_l, swing1LimitDegrees=45, swing2LimitDegrees=45, twistLimitDegrees=25}
 {bone1Name=dyn_jacketFront_03_l, bone2Name=dyn_jacketFront_02_l, swing1LimitDegrees=45, swing2LimitDegrees=45, twistLimitDegrees=25}
 {bone1Name=dyn_jacketFront_04_l, bone2Name=dyn_jacketFront_03_l, swing1LimitDegrees=45, swing2LimitDegrees=45, twistLimitDegrees=25}
 {bone1Name=dyn_jacketSide_01_r, bone2Name=dyn_jacketSide_mstr_r, swing1LimitDegrees=45, swing2LimitDegrees=45, twistLimitDegrees=5}
 {bone1Name=dyn_jacketSide_02_r, bone2Name=dyn_jacketSide_01_r, swing1LimitDegrees=45, swing2LimitDegrees=45, twistLimitDegrees=5}
 {bone1Name=dyn_jacketSide_03_r, bone2Name=dyn_jacketSide_02_r, swing1LimitDegrees=45, swing2LimitDegrees=45, twistLimitDegrees=5}
 {bone1Name=dyn_jacketSide_04_r, bone2Name=dyn_jacketSide_03_r, swing1LimitDegrees=45, swing2LimitDegrees=45, twistLimitDegrees=5}
 {bone1Name=dyn_jacketBack_01_r, bone2Name=dyn_jacketBack_mstr_r, swing1LimitDegrees=45, swing2LimitDegrees=45, twistLimitDegrees=5}
 {bone1Name=dyn_jacketBack_02_r, bone2Name=dyn_jacketBack_01_r, swing1LimitDegrees=10, swing2LimitDegrees=45, twistLimitDegrees=5}
 {bone1Name=dyn_jacketBack_03_r, bone2Name=dyn_jacketBack_02_r, swing1LimitDegrees=10, swing2LimitDegrees=45, twistLimitDegrees=5}
 {bone1Name=dyn_jacketBack_04_r, bone2Name=dyn_jacketBack_03_r, swing1LimitDegrees=10, swing2LimitDegrees=45, twistLimitDegrees=5}
 {bone1Name=dyn_jacketBack_01_l, bone2Name=dyn_jacketBack_mstr_l, swing1LimitDegrees=45, swing2LimitDegrees=45, twistLimitDegrees=5}
 {bone1Name=dyn_jacketBack_02_l, bone2Name=dyn_jacketBack_01_l, swing1LimitDegrees=45, swing2LimitDegrees=45, twistLimitDegrees=5}
 {bone1Name=dyn_jacketBack_03_l, bone2Name=dyn_jacketBack_02_l, swing1LimitDegrees=45, swing2LimitDegrees=45, twistLimitDegrees=5}
 {bone1Name=dyn_jacketBack_04_l, bone2Name=dyn_jacketBack_03_l, swing1LimitDegrees=45, swing2LimitDegrees=45, twistLimitDegrees=5}
 {bone1Name=dyn_jacketSide_01_l, bone2Name=dyn_jacketSide_mstr_l, swing1LimitDegrees=5, swing2LimitDegrees=45, twistLimitDegrees=5}
 {bone1Name=dyn_jacketSide_02_l, bone2Name=dyn_jacketSide_01_l, swing1LimitDegrees=5, swing2LimitDegrees=45, twistLimitDegrees=5}
 {bone1Name=dyn_jacketSide_03_l, bone2Name=dyn_jacketSide_02_l, swing1LimitDegrees=5, swing2LimitDegrees=45, twistLimitDegrees=5}
 {bone1Name=dyn_jacketSide_04_l, bone2Name=dyn_jacketSide_03_l, swing1LimitDegrees=5, swing2LimitDegrees=45, twistLimitDegrees=5}
 {
  bone1Name=pelvis
  bone2Name=root
  swing1LimitDegrees=45
  swing1Motion=Free
  swing2LimitDegrees=45
  swing2Motion=Free
  twistLimitDegrees=45
  twistMotion=Free
 }
]
