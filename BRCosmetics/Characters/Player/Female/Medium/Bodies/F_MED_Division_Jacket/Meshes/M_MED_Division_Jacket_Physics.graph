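# Physics asset for the Division Jacket player body cosmetic, providing collision and joints for the jacket's dynamic bones.
# Capsules cover pelvis, thighs and calves, with two spheres on root, one very large and offset.
# Front, side and back jacket chains, left and right, run four angle-limited segments from small master bodies, which join the pelvis with free rotation.
# Extra free-rotation constraints link upper right side segments to neighbouring front and back chains.
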PhysicsAsset /BRCosmetics/Characters/Player/Female/Medium/Bodies/F_MED_Division_Jacket/Meshes/M_MED_Division_Jacket_Physics
bodies {
 calf_l [
  {shapeType=Capsule, length=49.9672, radius=7.597, center=<-23.8904 -1.0428 -0.0001>, rotation=<rot 90 -18.9704 -24.9704>}
  {
   shapeType=Capsule
   length=38.6216
   radius=4.9618
   shapeName=calf_l_capsule
   center=<-14.97 5.6017 -0.0002>
   rotation=<rot 90 -18.9704 -24.9704>
  }
 ]
 calf_r [{shapeType=Capsule, length=52.6965, radius=14.0296, center=<30 2 -1.0863>, rotation=<rot 90 0 0>}]
 dyn_jacketBack_01_l [{shapeType=Capsule, length=22.0681, radius=3.355, shapeName=dyn_jacketBack_01_l_capsule, center=<6.247 -0.0002 0>, rotation=<rot 0 0 90>}]
 dyn_jacketBack_01_r [{shapeType=Capsule, length=7.7086, radius=2.2914, shapeName=dyn_jacketBack_01_r_capsule, center=<6.247 -0.0002 0>, rotation=<rot 0 0 90>}]
 dyn_jacketBack_02_l [{shapeType=Capsule, length=8.234, radius=6.6762, shapeName=dyn_jacketBack_02_l_capsule, center=<6.9678 0.0001 0.053>, rotation=<rot 90 0 0>}]
 dyn_jacketBack_02_r [{shapeType=Capsule, length=10.9757, radius=3.9345, shapeName=dyn_jacketBack_02_r_capsule, center=<6.9678 0.0001 0.053>, rotation=<rot 90 0 0>}]
 dyn_jacketBack_03_l [{shapeType=Capsule, length=16.1042, radius=8.0911, shapeName=dyn_jacketBack_03_l_capsule, center=<9.0532 0 0>, rotation=<rot 0 0 90>}]
 dyn_jacketBack_03_r [{shapeType=Capsule, length=19.6477, radius=5.9904, shapeName=dyn_jacketBack_03_r_capsule, center=<9.0532 0 0>, rotation=<rot 0 0 103.8549>}]
 dyn_jacketBack_04_l [{shapeType=Capsule, length=11.1619, radius=6.4266, shapeName=dyn_jacketBack_04_l_capsule, center=<6.247 -0.0002 0>, rotation=<rot 0 0 90>}]
 dyn_jacketBack_04_r [{
   shapeType=Capsule
   length=15.3501
   radius=8.1313
   shapeName=dyn_jacketBack_04_r_capsule
   center=<6.247 -0.0002 0>
   rotation=<rot 0 0 102.6162>
  }]
 dyn_jacketBack_mstr_l [{shapeType=Capsule, length=0.505, radius=0.505, rotation=<rot 0 -90 0>}]
 dyn_jacketBack_mstr_r [{shapeType=Capsule, length=0.505, radius=0.505, rotation=<rot 0 -90 0>}]
 dyn_jacketFront_01_l [{shapeType=Capsule, length=5.8666, radius=4.1334, shapeName=dyn_jacketFront_01_l_capsule, center=<7.821 2.4675 1.1944>, rotation=<rot 90 0 0>}]
 dyn_jacketFront_01_r [{
   shapeType=Capsule
   length=19.8047
   radius=6.3003
   shapeName=dyn_jacketFront_01_r_capsule
   center=<7.204 -3.2459 4.5581>
   rotation=<rot 2.4389 0.1531 54.5749>
  }]
 dyn_jacketFront_02_l [{
   shapeType=Capsule
   length=26.8456
   radius=4.1935
   shapeName=dyn_jacketFront_02_l_capsule
   center=<6.4596 0.0001 0>
   rotation=<rot 0 0 119.0208>
  }]
 dyn_jacketFront_02_r [{
   shapeType=Capsule
   length=36.776
   radius=5.0052
   shapeName=dyn_jacketFront_03_r_capsule
   center=<8.2668 -3.2343 1.9314>
   rotation=<rot 0 0 30.8457>
  }]
 dyn_jacketFront_03_l [{
   shapeType=Capsule
   length=27.0479
   radius=3.9911
   shapeName=dyn_jacketFront_03_l_capsule
   center=<6.247 -0.0001 0.2754>
   rotation=<rot 0 0 122.7366>
  }]
 dyn_jacketFront_03_r [{
   shapeType=Capsule
   length=32.3798
   radius=4.5238
   shapeName=dyn_jacketFront_03_r_capsule
   center=<8.2668 -0.0002 0>
   rotation=<rot 0 0 32.138>
  }]
 dyn_jacketFront_04_l [{
   shapeType=Capsule
   length=26.8579
   radius=4.1811
   shapeName=dyn_jacketFront_04_l_capsule
   center=<6.247 0 0.0798>
   rotation=<rot 0 0 119.7243>
  }]
 dyn_jacketFront_04_r [{
   shapeType=Capsule
   length=36.1249
   radius=4.2958
   shapeName=dyn_jacketFront_04_r_capsule
   center=<6.247 -0.0002 0>
   rotation=<rot 0 0 30.5575>
  }]
 dyn_jacketFront_mstr_l [{shapeType=Capsule, length=0.505, radius=0.505, rotation=<rot 0 -90 0>}]
 dyn_jacketFront_mstr_r [{shapeType=Capsule, length=0.505, radius=0.505, rotation=<rot 0 -90 0>}]
 dyn_jacketSide_01_l [{
   shapeType=Capsule
   length=20.6248
   radius=2.715
   shapeName=dyn_jacketSide_01_l_capsule
   center=<6.2471 -0.0003 -0.7228>
   rotation=<rot 90 -153.435 114.6603>
  }]
 dyn_jacketSide_01_r [{
   shapeType=Capsule
   length=16.6561
   radius=1.6398
   shapeName=dyn_jacketSide_01_r_capsule
   center=<6.674 -0.3678 2.104>
   rotation=<rot 90 180 93.1547>
  }]
 dyn_jacketSide_02_l [{shapeType=Capsule, length=19.4979, radius=3.5628, shapeName=dyn_jacketSide_02_l_capsule, center=<6.247 0.0001 0.053>, rotation=<rot 0 0 90>}]
 dyn_jacketSide_02_r [{
   shapeType=Capsule
   length=24.5536
   radius=2.3565
   shapeName=dyn_jacketSide_02_r_capsule
   center=<8.2353 0.0001 2.7137>
   rotation=<rot 0 0 90>
  }]
 dyn_jacketSide_03_l [{shapeType=Capsule, length=19.2964, radius=3.7643, shapeName=dyn_jacketSide_03_l_capsule, center=<6.247 -0.0002 0>, rotation=<rot 0 0 90>}]
 dyn_jacketSide_03_r [{shapeType=Capsule, length=14.2154, radius=2.8946, shapeName=dyn_jacketSide_03_r_capsule, center=<6.6728 0.0002 0>, rotation=<rot 0 0 90>}]
 dyn_jacketSide_04_l [{shapeType=Capsule, length=21.5581, radius=3.6999, shapeName=dyn_jacketSide_04_l_capsule, center=<6.247 -0.0002 0>, rotation=<rot 0 0 90>}]
 dyn_jacketSide_04_r [{shapeType=Capsule, length=17.5486, radius=3.3326, shapeName=dyn_jacketSide_04_r_capsule, center=<3.8377 0.0002 0>, rotation=<rot 0 0 90>}]
 dyn_jacketSide_mstr_l [{shapeType=Capsule, length=0.505, radius=0.505, rotation=<rot 0 -90 0>}]
 dyn_jacketSide_mstr_r [{shapeType=Capsule, length=0.505, radius=0.505, rotation=<rot 0 -90 0>}]
 pelvis [
  {shapeType=Capsule, length=12.7328, radius=10.5564, center=<-0.649 0.2801 -0.0002>, rotation=<rot 90 0 -3.3073>}
  {
   shapeType=Capsule
   length=14.2966
   radius=8.9927
   shapeName=pelvis_capsule
   center=<-0.2094 0.8359 -4.5558>
   rotation=<rot 80.001 -176.6924 -180>
  }
  {
   shapeType=Capsule
   length=12.7448
   radius=10.5444
   shapeName=pelvis_capsule
   center=<-0.2328 -1.3991 5.3434>
   rotation=<rot 77.8519 3.3074 0>
  }
 ]
 root [
  {shapeType=Sphere, radius=183.4375, shapeName=root_sphere, center=<0 -8.6403 -175.5073>}
  {shapeType=Sphere, radius=30.13, shapeName=root_sphere, center=<0.0002 2.4787 8.5972>}
 ]
 thigh_l [{shapeType=Capsule, length=28.7561, radius=7.2272, center=<-21.4904 1.0476 -0.2548>, rotation=<rot 90 12.9074 10.9407>}]
 thigh_r [{shapeType=Capsule, length=23.055, radius=12.9283, center=<20.804 0.0002 -0.5971>, rotation=<rot 84.8181 179.9814 179.9918>}]
}
constraints [
 {
  bone1Name=thigh_l
  bone2Name=pelvis
  swing1LimitDegrees=45
  swing1Motion=Free
  swing2LimitDegrees=45
  swing2Motion=Free
  twistLimitDegrees=45
  twistMotion=Free
 }
 {
  bone1Name=calf_l
  bone2Name=thigh_l
  swing1LimitDegrees=45
  swing1Motion=Free
  swing2LimitDegrees=45
  swing2Motion=Free
  twistLimitDegrees=45
  twistMotion=Free
 }
 {
  bone1Name=thigh_r
  bone2Name=pelvis
  swing1LimitDegrees=45
  swing1Motion=Free
  swing2LimitDegrees=45
  swing2Motion=Free
  twistLimitDegrees=45
  twistMotion=Free
 }
 {
  bone1Name=calf_r
  bone2Name=thigh_r
  swing1LimitDegrees=45
  swing1Motion=Free
  swing2LimitDegrees=45
  swing2Motion=Free
  twistLimitDegrees=45
  twistMotion=Free
 }
 {
  bone1Name=dyn_jacketSide_mstr_l
  bone2Name=pelvis
  swing1LimitDegrees=45
  swing1Motion=Free
  swing2LimitDegrees=45
  swing2Motion=Free
  twistLimitDegrees=45
  twistMotion=Free
 }
 {
  bone1Name=dyn_jacketBack_mstr_l
  bone2Name=pelvis
  swing1LimitDegrees=45
  swing1Motion=Free
  swing2LimitDegrees=45
  swing2Motion=Free
  twistLimitDegrees=45
  twistMotion=Free
 }
 {
  bone1Name=dyn_jacketBack_mstr_r
  bone2Name=pelvis
  swing1LimitDegrees=45
  swing1Motion=Free
  swing2LimitDegrees=45
  swing2Motion=Free
  twistLimitDegrees=45
  twistMotion=Free
 }
 {
  bone1Name=dyn_jacketSide_mstr_r
  bone2Name=pelvis
  swing1LimitDegrees=45
  swing1Motion=Free
  swing2LimitDegrees=45
  swing2Motion=Free
  twistLimitDegrees=45
  twistMotion=Free
 }
 {
  bone1Name=dyn_jacketFront_mstr_l
  bone2Name=pelvis
  swing1LimitDegrees=45
  swing1Motion=Free
  swing2LimitDegrees=45
  swing2Motion=Free
  twistLimitDegrees=45
  twistMotion=Free
 }
 {
  bone1Name=dyn_jacketFront_mstr_r
  bone2Name=pelvis
  swing1LimitDegrees=45
  swing1Motion=Free
  swing2LimitDegrees=45
  swing2Motion=Free
  twistLimitDegrees=45
  twistMotion=Free
 }
 {bone1Name=dyn_jacketFront_01_r, bone2Name=dyn_jacketFront_mstr_r, swing1LimitDegrees=30, swing2LimitDegrees=80, twistLimitDegrees=45}
 {bone1Name=dyn_jacketFront_02_r, bone2Name=dyn_jacketFront_01_r, swing1LimitDegrees=30, swing2LimitDegrees=80, twistLimitDegrees=45}
 {bone1Name=dyn_jacketFront_03_r, bone2Name=dyn_jacketFront_02_r, swing1LimitDegrees=30, swing2LimitDegrees=80, twistLimitDegrees=45}
 {bone1Name=dyn_jacketFront_04_r, bone2Name=dyn_jacketFront_03_r, swing1LimitDegrees=30, swing2LimitDegrees=80, twistLimitDegrees=45}
 {bone1Name=dyn_jacketFront_01_l, bone2Name=dyn_jacketFront_mstr_l, swing1LimitDegrees=45, swing2LimitDegrees=45, twistLimitDegrees=25}
 {bone1Name=dyn_jacketFront_02_l, bone2Name=dyn_jacketFront_01_l, swing1LimitDegrees=45, swing2LimitDegrees=45, twistLimitDegrees=25}
 {bone1Name=dyn_jacketFront_03_l, bone2Name=dyn_jacketFront_02_l, swing1LimitDegrees=45, swing2LimitDegrees=45, twistLimitDegrees=25}
 {bone1Name=dyn_jacketFront_04_l, bone2Name=dyn_jacketFront_03_l, swing1LimitDegrees=45, swing2LimitDegrees=45, twistLimitDegrees=25}
 {bone1Name=dyn_jacketSide_01_r, bone2Name=dyn_jacketSide_mstr_r, swing1LimitDegrees=20, swing2LimitDegrees=45, twistLimitDegrees=50}
 {bone1Name=dyn_jacketSide_02_r, bone2Name=dyn_jacketSide_01_r, swing1LimitDegrees=20, swing2LimitDegrees=45, twistLimitDegrees=50}
 {bone1Name=dyn_jacketSide_03_r, bone2Name=dyn_jacketSide_02_r, swing1LimitDegrees=20, swing2LimitDegrees=45, twistLimitDegrees=50}
 {bone1Name=dyn_jacketSide_04_r, bone2Name=dyn_jacketSide_03_r, swing1LimitDegrees=20, swing2LimitDegrees=45, twistLimitDegrees=50}
 {bone1Name=dyn_jacketBack_01_r, bone2Name=dyn_jacketBack_mstr_r, swing1LimitDegrees=45, swing2LimitDegrees=45, twistLimitDegrees=5}
 {bone1Name=dyn_jacketBack_02_r, bone2Name=dyn_jacketBack_01_r, swing1LimitDegrees=45, swing2LimitDegrees=45, twistLimitDegrees=5}
 {bone1Name=dyn_jacketBack_03_r, bone2Name=dyn_jacketBack_02_r, swing1LimitDegrees=45, swing2LimitDegrees=45, twistLimitDegrees=5}
 {bone1Name=dyn_jacketBack_04_r, bone2Name=dyn_jacketBack_03_r, swing1LimitDegrees=45, swing2LimitDegrees=45, twistLimitDegrees=5}
 {bone1Name=dyn_jacketBack_01_l, bone2Name=dyn_jacketBack_mstr_l, swing1LimitDegrees=10, swing2LimitDegrees=45, twistLimitDegrees=5}
 {bone1Name=dyn_jacketBack_02_l, bone2Name=dyn_jacketBack_01_l, swing1LimitDegrees=45, swing2LimitDegrees=45, twistLimitDegrees=5}
 {bone1Name=dyn_jacketBack_03_l, bone2Name=dyn_jacketBack_02_l, swing1LimitDegrees=45, swing2LimitDegrees=45, twistLimitDegrees=5}
 {bone1Name=dyn_jacketBack_04_l, bone2Name=dyn_jacketBack_03_l, swing1LimitDegrees=45, swing2LimitDegrees=45, twistLimitDegrees=5}
 {bone1Name=dyn_jacketSide_01_l, bone2Name=dyn_jacketSide_mstr_l, swing1LimitDegrees=45, swing2LimitDegrees=45, twistLimitDegrees=45}
 {bone1Name=dyn_jacketSide_02_l, bone2Name=dyn_jacketSide_01_l, swing1LimitDegrees=45, swing2LimitDegrees=45, twistLimitDegrees=45}
 {bone1Name=dyn_jacketSide_03_l, bone2Name=dyn_jacketSide_02_l, swing1LimitDegrees=45, swing2LimitDegrees=45, twistLimitDegrees=45}
 {bone1Name=dyn_jacketSide_04_l, bone2Name=dyn_jacketSide_03_l, swing1LimitDegrees=45, swing2LimitDegrees=45, twistLimitDegrees=45}
 {
  bone1Name=pelvis
  bone2Name=root
  swing1LimitDegrees=45
  swing1Motion=Free
  swing2LimitDegrees=45
  swing2Motion=Free
  twistLimitDegrees=45
  twistMotion=Free
 }
 {
  bone1Name=dyn_jacketSide_01_r
  bone2Name=dyn_jacketBack_01_r
  swing1LimitDegrees=45
  swing1Motion=Free
  swing2LimitDegrees=45
  swing2Motion=Free
  twistLimitDegrees=45
  twistMotion=Free
 }
 {
  bone1Name=dyn_jacketSide_02_r
  bone2Name=dyn_jacketBack_02_r
  swing1LimitDegrees=45
  swing1Motion=Free
  swing2LimitDegrees=45
  swing2Motion=Free
  twistLimitDegrees=45
  twistMotion=Free
 }
 {
  bone1Name=dyn_jacketSide_01_r
  bone2Name=dyn_jacketFront_01_r
  swing1LimitDegrees=45
  swing1Motion=Free
  swing2LimitDegrees=45
  swing2Motion=Free
  twistLimitDegrees=45
  twistMotion=Free
 }
]
