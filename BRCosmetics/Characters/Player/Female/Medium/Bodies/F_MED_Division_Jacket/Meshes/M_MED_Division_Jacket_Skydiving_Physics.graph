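# Physics asset for the Division Jacket player body's skydiving variant, covering the jacket's dynamic bones.
# Capsules on pelvis, thighs and calves, plus two root spheres, form the body shapes.
# Front, side and back four-link chains per side hang from pelvis-linked master bones, with tight twist and swing mostly looser toward back and side tips.
# Only left chains are cross-linked by angularly free constraints: back-to-side fully, front-to-side at the upper two links.
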PhysicsAsset /BRCosmetics/Characters/Player/Female/Medium/Bodies/F_MED_Division_Jacket/Meshes/M_MED_Division_Jacket_Skydiving_Physics
bodies {
 calf_l [{shapeType=Capsule, length=48.1039, radius=12.6324, center=<-23.8904 -1.0428 -0.0001>, rotation=<rot 84.245 -174.0002 180>}]
 calf_r [{shapeType=Capsule, length=54.5943, radius=15.304, center=<30 2 -1.0863>, rotation=<rot 90 0 0>}]
 dyn_jacketBack_01_l [{shapeType=Capsule, length=7.8918, radius=2.1082, shapeName=dyn_jacketBack_01_l_capsule, center=<6.247 -0.0002 0>, rotation=<rot 0 0 90>}]
 dyn_jacketBack_01_r [{shapeType=Capsule, length=7.7086, radius=2.2914, shapeName=dyn_jacketBack_01_r_capsule, center=<6.247 -0.0002 0>, rotation=<rot 0 0 90>}]
 dyn_jacketBack_02_l [{shapeType=Capsule, length=10.3837, radius=4.5264, shapeName=dyn_jacketBack_02_l_capsule, center=<6.9678 0.0001 0.053>, rotation=<rot 90 0 0>}]
 dyn_jacketBack_02_r [{shapeType=Capsule, length=10.3837, radius=4.5264, shapeName=dyn_jacketBack_02_r_capsule, center=<6.9678 0.0001 0.053>, rotation=<rot 90 0 0>}]
 dyn_jacketBack_03_l [{shapeType=Capsule, length=8.3627, radius=6.6206, shapeName=dyn_jacketBack_03_l_capsule, center=<6.247 -0.0002 0>, rotation=<rot 90 0 0>}]
 dyn_jacketBack_03_r [{
   shapeType=Capsule
   length=9.4519
   radius=5.5314
   shapeName=dyn_jacketBack_03_r_capsule
   center=<7.5496 -0.0379 -0.1663>
   rotation=<rot 81.0618 148.4992 171.009>
  }]
 dyn_jacketBack_04_l [{shapeType=Capsule, length=12.3651, radius=5.2235, shapeName=dyn_jacketBack_04_l_capsule, center=<6.247 -0.0002 0>, rotation=<rot 0 0 90>}]
 dyn_jacketBack_04_r [{shapeType=Capsule, length=12.3938, radius=5.1947, shapeName=dyn_jacketBack_04_r_capsule, center=<6.247 -0.0002 0>, rotation=<rot 0 0 90>}]
 dyn_jacketBack_mstr_l [{shapeType=Capsule, length=0.505, radius=0.505, rotation=<rot 0 -90 0>}]
 dyn_jacketBack_mstr_r [{shapeType=Capsule, length=0.505, radius=0.505, rotation=<rot 0 -90 0>}]
 dyn_jacketFront_01_l [{shapeType=Capsule, length=7.4509, radius=2.5491, shapeName=dyn_jacketFront_01_l_capsule, center=<7.821 0 0>, rotation=<rot 90 0 0>}]
 dyn_jacketFront_01_r [{shapeType=Capsule, length=9.1605, radius=0.8395, shapeName=dyn_jacketFront_01_r_capsule, center=<8.2194 0 0>, rotation=<rot 90 0 0>}]
 dyn_jacketFront_02_l [{
   shapeType=Capsule
   length=16.0052
   radius=5.7777
   shapeName=dyn_jacketFront_02_l_capsule
   center=<6.4596 0.0001 0>
   rotation=<rot 0 0 126.6752>
  }]
 dyn_jacketFront_02_r [{
   shapeType=Capsule
   length=22.2023
   radius=3.2776
   shapeName=dyn_jacketFront_02_r_capsule
   center=<6.247 -0.0002 0>
   rotation=<rot 0 0 49.0825>
  }]
 dyn_jacketFront_03_l [{
   shapeType=Capsule
   length=16.7032
   radius=5.0797
   shapeName=dyn_jacketFront_03_l_capsule
   center=<6.247 -0.0001 0.2754>
   rotation=<rot 0 0 130.3902>
  }]
 dyn_jacketFront_03_r [{
   shapeType=Capsule
   length=25.1409
   radius=4.2981
   shapeName=dyn_jacketFront_03_r_capsule
   center=<8.2668 -0.0002 0>
   rotation=<rot 0 0 49.9708>
  }]
 dyn_jacketFront_04_l [{
   shapeType=Capsule
   length=15.8369
   radius=5.946
   shapeName=dyn_jacketFront_04_l_capsule
   center=<6.247 0 0.0798>
   rotation=<rot 0 0 127.3632>
  }]
 dyn_jacketFront_04_r [{
   shapeType=Capsule
   length=20.9562
   radius=4.5236
   shapeName=dyn_jacketFront_04_r_capsule
   center=<6.247 -0.0002 0>
   rotation=<rot 0 0 48.3905>
  }]
 dyn_jacketFront_mstr_l [{shapeType=Capsule, length=0.505, radius=0.505, rotation=<rot 0 -90 0>}]
 dyn_jacketFront_mstr_r [{shapeType=Capsule, length=0.505, radius=0.505, rotation=<rot 0 -90 0>}]
 dyn_jacketSide_01_l [{
   shapeType=Capsule
   length=11.1485
   radius=1.9117
   shapeName=dyn_jacketSide_01_l_capsule
   center=<6.2471 -0.0003 0.0912>
   rotation=<rot 90 -56.3099 -148.2146>
  }]
 dyn_jacketSide_01_r [{
   shapeType=Capsule
   length=22.5774
   radius=2.5879
   shapeName=dyn_jacketSide_01_r_capsule
   center=<8.4049 -0.1193 1.1086>
   rotation=<rot 90 0 -86.8453>
  }]
 dyn_jacketSide_02_l [{
   shapeType=Capsule
   length=15.1795
   radius=6.2937
   shapeName=dyn_jacketSide_02_l_capsule
   center=<6.247 -2.301 5.7296>
   rotation=<rot 0 0 67.9335>
  }]
 dyn_jacketSide_02_r [{shapeType=Capsule, length=31.5888, radius=3.828, shapeName=dyn_jacketSide_02_r_capsule, center=<8.2353 0.0001 0.053>, rotation=<rot 0 0 90>}]
 dyn_jacketSide_03_l [{
   shapeType=Capsule
   length=14.8876
   radius=6.5856
   shapeName=dyn_jacketSide_03_l_capsule
   center=<6.247 -0.0002 3.2623>
   rotation=<rot 0 0 90>
  }]
 dyn_jacketSide_03_r [{shapeType=Capsule, length=30.6163, radius=4.8004, shapeName=dyn_jacketSide_03_r_capsule, center=<6.6728 0.0002 0>, rotation=<rot 0 0 90>}]
 dyn_jacketSide_04_l [{
   shapeType=Capsule
   length=17.1493
   radius=6.5212
   shapeName=dyn_jacketSide_04_l_capsule
   center=<6.247 -0.0002 3.2623>
   rotation=<rot 0 0 90>
  }]
 dyn_jacketSide_04_r [{shapeType=Capsule, length=31.4098, radius=4.0069, shapeName=dyn_jacketSide_04_r_capsule, center=<3.8377 0.0002 0>, rotation=<rot 0 0 90>}]
 dyn_jacketSide_mstr_l [{shapeType=Capsule, length=0.505, radius=0.505, rotation=<rot 0 -90 0>}]
 dyn_jacketSide_mstr_r [{shapeType=Capsule, length=0.505, radius=0.505, rotation=<rot 0 -90 0>}]
 pelvis [
  {shapeType=Capsule, length=12.7328, radius=10.5564, center=<-0.649 0.2801 -0.0002>, rotation=<rot 90 0 -3.3073>}
  {
   shapeType=Capsule
   length=15.3785
   radius=7.9107
   shapeName=pelvis_capsule
   center=<-0.0808 -1.3903 -4.5558>
   rotation=<rot 80.001 -176.6924 -180>
  }
  {shapeType=Capsule, length=15.6212, radius=7.6681, shapeName=pelvis_capsule, center=<-0.2328 -1.3991 5.3434>, rotation=<rot 77.8519 3.3074 0>}
 ]
 root [
  {shapeType=Sphere, radius=183.4375, shapeName=root_sphere, center=<0 -8.6403 -175.5073>}
  {shapeType=Sphere, radius=30.13, shapeName=root_sphere, center=<0.0002 2.4787 8.5972>}
 ]
 thigh_l [{shapeType=Capsule, length=26.1249, radius=9.8584, center=<-21.4337 0 2.1937>, rotation=<rot 86.8469 179.9999 180>}]
 thigh_r [{shapeType=Capsule, length=30.1826, radius=9.8584, center=<21.7868 0 -2.6951>, rotation=<rot 86.4316 179.9999 179.9999>}]
}
constraints [
 {
  bone1Name=thigh_l
  bone2Name=pelvis
  swing1LimitDegrees=45
  swing1Motion=Free
  swing2LimitDegrees=45
  swing2Motion=Free
  twistLimitDegrees=45
  twistMotion=Free
 }
 {
  bone1Name=calf_l
  bone2Name=thigh_l
  swing1LimitDegrees=45
  swing1Motion=Free
  swing2LimitDegrees=45
  swing2Motion=Free
  twistLimitDegrees=45
  twistMotion=Free
 }
 {
  bone1Name=thigh_r
  bone2Name=pelvis
  swing1LimitDegrees=45
  swing1Motion=Free
  swing2LimitDegrees=45
  swing2Motion=Free
  twistLimitDegrees=45
  twistMotion=Free
 }
 {
  bone1Name=calf_r
  bone2Name=thigh_r
  swing1LimitDegrees=45
  swing1Motion=Free
  swing2LimitDegrees=45
  swing2Motion=Free
  twistLimitDegrees=45
  twistMotion=Free
 }
 {
  bone1Name=dyn_jacketSide_mstr_l
  bone2Name=pelvis
  swing1LimitDegrees=45
  swing1Motion=Free
  swing2LimitDegrees=45
  swing2Motion=Free
  twistLimitDegrees=45
  twistMotion=Free
 }
 {
  bone1Name=dyn_jacketBack_mstr_l
  bone2Name=pelvis
  swing1LimitDegrees=45
  swing1Motion=Free
  swing2LimitDegrees=45
  swing2Motion=Free
  twistLimitDegrees=45
  twistMotion=Free
 }
 {
  bone1Name=dyn_jacketBack_mstr_r
  bone2Name=pelvis
  swing1LimitDegrees=45
  swing1Motion=Free
  swing2LimitDegrees=45
  swing2Motion=Free
  twistLimitDegrees=45
  twistMotion=Free
 }
 {
  bone1Name=dyn_jacketSide_mstr_r
  bone2Name=pelvis
  swing1LimitDegrees=45
  swing1Motion=Free
  swing2LimitDegrees=45
  swing2Motion=Free
  twistLimitDegrees=45
  twistMotion=Free
 }
 {
  bone1Name=dyn_jacketFront_mstr_l
  bone2Name=pelvis
  swing1LimitDegrees=45
  swing1Motion=Free
  swing2LimitDegrees=45
  swing2Motion=Free
  twistLimitDegrees=45
  twistMotion=Free
 }
 {
  bone1Name=dyn_jacketFront_mstr_r
  bone2Name=pelvis
  swing1LimitDegrees=45
  swing1Motion=Free
  swing2LimitDegrees=45
  swing2Motion=Free
  twistLimitDegrees=45
  twistMotion=Free
 }
 {bone1Name=dyn_jacketFront_01_r, bone2Name=dyn_jacketFront_mstr_r, swing1LimitDegrees=15, swing2LimitDegrees=15, twistLimitDegrees=5}
 {bone1Name=dyn_jacketFront_02_r, bone2Name=dyn_jacketFront_01_r, swing1LimitDegrees=15, swing2LimitDegrees=15, twistLimitDegrees=5}
 {bone1Name=dyn_jacketFront_03_r, bone2Name=dyn_jacketFront_02_r, swing1LimitDegrees=15, swing2LimitDegrees=15, twistLimitDegrees=5}
 {bone1Name=dyn_jacketFront_04_r, bone2Name=dyn_jacketFront_03_r, swing1LimitDegrees=15, swing2LimitDegrees=15, twistLimitDegrees=5}
 {bone1Name=dyn_jacketFront_01_l, bone2Name=dyn_jacketFront_mstr_l, swing1LimitDegrees=15, swing2LimitDegrees=15, twistLimitDegrees=5}
 {bone1Name=dyn_jacketFront_02_l, bone2Name=dyn_jacketFront_01_l, swing1LimitDegrees=15, swing2LimitDegrees=15, twistLimitDegrees=5}
 {bone1Name=dyn_jacketFront_03_l, bone2Name=dyn_jacketFront_02_l, swing1LimitDegrees=15, swing2LimitDegrees=15, twistLimitDegrees=5}
 {bone1Name=dyn_jacketFront_04_l, bone2Name=dyn_jacketFront_03_l, swing1LimitDegrees=45, swing2LimitDegrees=45, twistLimitDegrees=5}
 {bone1Name=dyn_jacketSide_01_r, bone2Name=dyn_jacketSide_mstr_r, swing1LimitDegrees=15, swing2LimitDegrees=15, twistLimitDegrees=5}
 {bone1Name=dyn_jacketSide_02_r, bone2Name=dyn_jacketSide_01_r, swing1LimitDegrees=15, swing2LimitDegrees=15, twistLimitDegrees=5}
 {bone1Name=dyn_jacketSide_03_r, bone2Name=dyn_jacketSide_02_r, swing1LimitDegrees=45, swing2LimitDegrees=45, twistLimitDegrees=5}
 {bone1Name=dyn_jacketSide_04_r, bone2Name=dyn_jacketSide_03_r, swing1LimitDegrees=45, swing2LimitDegrees=45, twistLimitDegrees=5}
 {bone1Name=dyn_jacketBack_01_r, bone2Name=dyn_jacketBack_mstr_r, swing1LimitDegrees=15, swing2LimitDegrees=15, twistLimitDegrees=5}
 {bone1Name=dyn_jacketBack_02_r, bone2Name=dyn_jacketBack_01_r, swing1LimitDegrees=15, swing2LimitDegrees=15, twistLimitDegrees=5}
 {bone1Name=dyn_jacketBack_03_r, bone2Name=dyn_jacketBack_02_r, swing1LimitDegrees=45, swing2LimitDegrees=45, twistLimitDegrees=5}
 {bone1Name=dyn_jacketBack_04_r, bone2Name=dyn_jacketBack_03_r, swing1LimitDegrees=45, swing2LimitDegrees=45, twistLimitDegrees=5}
 {bone1Name=dyn_jacketBack_01_l, bone2Name=dyn_jacketBack_mstr_l, swing1LimitDegrees=15, swing2LimitDegrees=15, twistLimitDegrees=5}
 {bone1Name=dyn_jacketBack_02_l, bone2Name=dyn_jacketBack_01_l, swing1LimitDegrees=15, swing2LimitDegrees=15, twistLimitDegrees=5}
 {bone1Name=dyn_jacketBack_03_l, bone2Name=dyn_jacketBack_02_l, swing1LimitDegrees=45, swing2LimitDegrees=45, twistLimitDegrees=5}
 {bone1Name=dyn_jacketBack_04_l, bone2Name=dyn_jacketBack_03_l, swing1LimitDegrees=45, swing2LimitDegrees=45, twistLimitDegrees=5}
 {bone1Name=dyn_jacketSide_01_l, bone2Name=dyn_jacketSide_mstr_l, swing1LimitDegrees=15, swing2LimitDegrees=15, twistLimitDegrees=5}
 {bone1Name=dyn_jacketSide_02_l, bone2Name=dyn_jacketSide_01_l, swing1LimitDegrees=15, swing2LimitDegrees=15, twistLimitDegrees=5}
 {bone1Name=dyn_jacketSide_03_l, bone2Name=dyn_jacketSide_02_l, swing1LimitDegrees=45, swing2LimitDegrees=45, twistLimitDegrees=5}
 {bone1Name=dyn_jacketSide_04_l, bone2Name=dyn_jacketSide_03_l, swing1LimitDegrees=45, swing2LimitDegrees=45, twistLimitDegrees=5}
 {
  bone1Name=pelvis
  bone2Name=root
  swing1LimitDegrees=45
  swing1Motion=Free
  swing2LimitDegrees=45
  swing2Motion=Free
  twistLimitDegrees=45
  twistMotion=Free
 }
 {
  bone1Name=dyn_jacketSide_01_l
  bone2Name=dyn_jacketBack_01_l
  swing1LimitDegrees=45
  swing1Motion=Free
  swing2LimitDegrees=45
  swing2Motion=Free
  twistLimitDegrees=45
  twistMotion=Free
 }
 {
  bone1Name=dyn_jacketSide_02_l
  bone2Name=dyn_jacketBack_02_l
  swing1LimitDegrees=45
  swing1Motion=Free
  swing2LimitDegrees=45
  swing2Motion=Free
  twistLimitDegrees=45
  twistMotion=Free
 }
 {
  bone1Name=dyn_jacketSide_03_l
  bone2Name=dyn_jacketBack_03_l
  swing1LimitDegrees=45
  swing1Motion=Free
  swing2LimitDegrees=45
  swing2Motion=Free
  twistLimitDegrees=45
  twistMotion=Free
 }
 {
  bone1Name=dyn_jacketSide_04_l
  bone2Name=dyn_jacketBack_04_l
  swing1LimitDegrees=45
  swing1Motion=Free
  swing2LimitDegrees=45
  swing2Motion=Free
  twistLimitDegrees=45
  twistMotion=Free
 }
 {
  bone1Name=dyn_jacketFront_01_l
  bone2Name=dyn_jacketSide_01_l
  swing1LimitDegrees=45
  swing1Motion=Free
  swing2LimitDegrees=45
  swing2Motion=Free
  twistLimitDegrees=45
  twistMotion=Free
 }
 {
  bone1Name=dyn_jacketFront_02_l
  bone2Name=dyn_jacketSide_02_l
  swing1LimitDegrees=45
  swing1Motion=Free
  swing2LimitDegrees=45
  swing2Motion=Free
  twistLimitDegrees=45
  twistMotion=Free
 }
]
