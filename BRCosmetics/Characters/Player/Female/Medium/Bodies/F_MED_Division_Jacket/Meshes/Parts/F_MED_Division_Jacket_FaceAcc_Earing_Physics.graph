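# Physics asset for the earring face accessory of the female medium Division Jacket character, setting up bodies for dangling earring simulation.
# Each side has a two-bone capsule chain, with the first link constrained to the head and the second to the first.
# All four joints share the same wide swing and narrower twist limits.
# The head body is built from three overlapping capsules that approximate the skull's shape.
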PhysicsAsset /BRCosmetics/Characters/Player/Female/Medium/Bodies/F_MED_Division_Jacket/Meshes/Parts/F_MED_Division_Jacket_FaceAcc_Earing_Physics
bodies {
 dyn_l_earing_01 [{
   shapeType=Capsule
   length=0.505
   radius=0.505
   shapeName=dyn_l_earing_01_capsule
   center=<0.5801 0.0061 0.1884>
   rotation=<rot 0 91.7585 -87.498>
  }]
 dyn_l_earing_02 [{
   shapeType=Capsule
   length=2.5879
   radius=0.994
   shapeName=dyn_l_earing_02_capsule
   center=<2.656 -0.0321 0.1952>
   rotation=<rot 0 89.5941 -89.9623>
  }]
 dyn_r_earing_01 [{
   shapeType=Capsule
   length=0.505
   radius=0.505
   shapeName=dyn_r_earing_01_capsule
   center=<0.3986 0.0269 -0.0011>
   rotation=<rot 0 88.2414 -87.4981>
  }]
 dyn_r_earing_02 [{
   shapeType=Capsule
   length=2.5879
   radius=0.994
   shapeName=dyn_r_earing_02_capsule
   center=<2.4662 0.0582 0.0136>
   rotation=<rot 0 90.4059 -89.9624>
  }]
 head [
  {shapeType=Capsule, length=4.0872, radius=5.2469, shapeName=head_capsule, rotation=<rot 90 0 0>}
  {shapeType=Capsule, length=3.7766, radius=5.5575, shapeName=head_capsule, center=<1.0557 4.1532 0.0003>, rotation=<rot 90 0 0>}
  {shapeType=Capsule, length=1.9356, radius=6.2765, shapeName=head_capsule, center=<4.7915 2.8685 0.0001>, rotation=<rot 90 0.0066 0.0066>}
 ]
}
constraints [
 {bone1Name=dyn_r_earing_01, bone2Name=head, swing1LimitDegrees=80, swing2LimitDegrees=80, twistLimitDegrees=30}
 {bone1Name=dyn_r_earing_02, bone2Name=dyn_r_earing_01, swing1LimitDegrees=80, swing2LimitDegrees=80, twistLimitDegrees=30}
 {bone1Name=dyn_l_earing_01, bone2Name=head, swing1LimitDegrees=80, swing2LimitDegrees=80, twistLimitDegrees=30}
 {bone1Name=dyn_l_earing_02, bone2Name=dyn_l_earing_01, swing1LimitDegrees=80, swing2LimitDegrees=80, twistLimitDegrees=30}
]
