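# Physics asset for the hair part of the female medium Division Jacket character, defining dynamics for a bun and four bangs.
# Each bang is a four-capsule bone chain with tight twist and wider swing on one axis, rooted in a small MSTR body.
# The bun is a single capsule limited against its own MSTR body.
# All MSTR bodies join the head with free angular motion, and three overlapping capsules approximate the head's shape.
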PhysicsAsset /BRCosmetics/Characters/Player/Female/Medium/Bodies/F_MED_Division_Jacket/Meshes/Parts/F_MED_Division_Jacket_FaceAcc_Hair_Physics
bodies {
 dyn_bun [{shapeType=Capsule, length=0, radius=4.6139, shapeName=dyn_bun_capsule, center=<2.2283 -0.0002 0>}]
 dyn_bun_MSTR [{shapeType=Capsule, length=0.505, radius=0.505, shapeName=dyn_bun_MSTR_capsule}]
 dyn_hairBang_a_01 [{shapeType=Capsule, length=1.3117, radius=0.6883, shapeName=dyn_hairBang_a_01_capsule, center=<0.9398 0 0>, rotation=<rot 0 0 90>}]
 dyn_hairBang_a_02 [{shapeType=Capsule, length=1.3117, radius=0.6883, shapeName=dyn_hairBang_a_02_capsule, center=<0.9398 0 0>, rotation=<rot 0 0 90>}]
 dyn_hairBang_a_03 [{shapeType=Capsule, length=1.3117, radius=0.6883, shapeName=dyn_hairBang_a_03_capsule, center=<0.9398 0 0>, rotation=<rot 0 0 90>}]
 dyn_hairBang_a_04 [{shapeType=Capsule, length=1.3117, radius=0.6883, shapeName=dyn_hairBang_a_04_capsule, center=<0.9398 0 0>, rotation=<rot 0 0 90>}]
 dyn_hairBang_a_MSTR [{shapeType=Capsule, length=0.505, radius=0.505, shapeName=dyn_hairBang_a_MSTR_capsule}]
 dyn_hairBang_b_01 [{shapeType=Capsule, length=1.3117, radius=0.6883, shapeName=dyn_hairBang_b_01_capsule, center=<0.9398 0 0>, rotation=<rot 0 0 90>}]
 dyn_hairBang_b_02 [{shapeType=Capsule, length=1.3117, radius=0.6883, shapeName=dyn_hairBang_b_02_capsule, center=<0.9398 0 0>, rotation=<rot 0 0 90>}]
 dyn_hairBang_b_03 [{shapeType=Capsule, length=1.3117, radius=0.6883, shapeName=dyn_hairBang_b_03_capsule, center=<0.9398 0 0>, rotation=<rot 0 0 90>}]
 dyn_hairBang_b_04 [{shapeType=Capsule, length=1.3117, radius=0.6883, shapeName=dyn_hairBang_b_04_capsule, center=<0.9398 0 0>, rotation=<rot 0 0 90>}]
 dyn_hairBang_b_MSTR [{shapeType=Capsule, length=0.505, radius=0.505, shapeName=dyn_hairBang_b_MSTR_capsule}]
 dyn_hairBang_c_01 [{shapeType=Capsule, length=1.3117, radius=0.6883, shapeName=dyn_hairBang_c_01_capsule, center=<0.9398 0 0>, rotation=<rot 0 0 90>}]
 dyn_hairBang_c_02 [{shapeType=Capsule, length=1.3117, radius=0.6883, shapeName=dyn_hairBang_c_02_capsule, center=<0.9398 0 0>, rotation=<rot 0 0 90>}]
 dyn_hairBang_c_03 [{shapeType=Capsule, length=1.3117, radius=0.6883, shapeName=dyn_hairBang_c_03_capsule, center=<0.9398 0 0>, rotation=<rot 0 0 90>}]
 dyn_hairBang_c_04 [{shapeType=Capsule, length=1.3117, radius=0.6883, shapeName=dyn_hairBang_c_04_capsule, center=<0.9398 0 0>, rotation=<rot 0 0 90>}]
 dyn_hairBang_c_MSTR [{shapeType=Capsule, length=0.505, radius=0.505, shapeName=dyn_hairBang_c_MSTR_capsule}]
 dyn_hairBang_d_01 [{shapeType=Capsule, length=1.4958, radius=0.5042, shapeName=dyn_hairBang_d_01_capsule, center=<0.9398 0 0>, rotation=<rot 0 0 90>}]
 dyn_hairBang_d_02 [{shapeType=Capsule, length=1.4958, radius=0.5042, shapeName=dyn_hairBang_d_02_capsule, center=<0.9398 0 0>, rotation=<rot 0 0 90>}]
 dyn_hairBang_d_03 [{shapeType=Capsule, length=1.4783, radius=0.5217, shapeName=dyn_hairBang_d_03_capsule, center=<0.9398 0 0>, rotation=<rot 0 0 90>}]
 dyn_hairBang_d_04 [{shapeType=Capsule, length=1.4783, radius=0.5217, shapeName=dyn_hairBang_d_04_capsule, center=<0.9398 0 0>, rotation=<rot 0 0 90>}]
 dyn_hairBang_d_MSTR [{shapeType=Capsule, length=0.505, radius=0.505, shapeName=dyn_hairBang_d_MSTR_capsule}]
 head [
  {shapeType=Capsule, length=4.0872, radius=5.2469, shapeName=head_capsule, rotation=<rot 90 0 0>}
  {shapeType=Capsule, length=3.7766, radius=5.5575, shapeName=head_capsule, center=<1.0557 4.1532 0.0003>, rotation=<rot 90 0 0>}
  {shapeType=Capsule, length=1.9356, radius=6.2765, shapeName=head_capsule, center=<4.7915 2.8685 0.0001>, rotation=<rot 90 0.0066 0.0066>}
 ]
}
constraints [
 {bone1Name=dyn_bun, bone2Name=dyn_bun_MSTR, swing1LimitDegrees=10, swing2LimitDegrees=10, twistLimitDegrees=10}
 {
  bone1Name=dyn_bun_MSTR
  bone2Name=head
  swing1LimitDegrees=45
  swing1Motion=Free
  swing2LimitDegrees=45
  swing2Motion=Free
  twistLimitDegrees=45
  twistMotion=Free
 }
 {
  bone1Name=dyn_hairBang_d_MSTR
  bone2Name=head
  swing1LimitDegrees=45
  swing1Motion=Free
  swing2LimitDegrees=45
  swing2Motion=Free
  twistLimitDegrees=45
  twistMotion=Free
 }
 {
  bone1Name=dyn_hairBang_c_MSTR
  bone2Name=head
  swing1LimitDegrees=45
  swing1Motion=Free
  swing2LimitDegrees=45
  swing2Motion=Free
  twistLimitDegrees=45
  twistMotion=Free
 }
 {
  bone1Name=dyn_hairBang_b_MSTR
  bone2Name=head
  swing1LimitDegrees=45
  swing1Motion=Free
  swing2LimitDegrees=45
  swing2Motion=Free
  twistLimitDegrees=45
  twistMotion=Free
 }
 {
  bone1Name=dyn_hairBang_a_MSTR
  bone2Name=head
  swing1LimitDegrees=45
  swing1Motion=Free
  swing2LimitDegrees=45
  swing2Motion=Free
  twistLimitDegrees=45
  twistMotion=Free
 }
 {bone1Name=dyn_hairBang_d_01, bone2Name=dyn_hairBang_d_MSTR, swing1LimitDegrees=10, swing2LimitDegrees=30, twistLimitDegrees=5}
 {bone1Name=dyn_hairBang_d_02, bone2Name=dyn_hairBang_d_01, swing1LimitDegrees=10, swing2LimitDegrees=30, twistLimitDegrees=5}
 {bone1Name=dyn_hairBang_d_03, bone2Name=dyn_hairBang_d_02, swing1LimitDegrees=10, swing2LimitDegrees=30, twistLimitDegrees=5}
 {bone1Name=dyn_hairBang_d_04, bone2Name=dyn_hairBang_d_03, swing1LimitDegrees=10, swing2LimitDegrees=30, twistLimitDegrees=5}
 {bone1Name=dyn_hairBang_c_01, bone2Name=dyn_hairBang_c_MSTR, swing1LimitDegrees=10, swing2LimitDegrees=30, twistLimitDegrees=5}
 {bone1Name=dyn_hairBang_c_02, bone2Name=dyn_hairBang_c_01, swing1LimitDegrees=10, swing2LimitDegrees=30, twistLimitDegrees=5}
 {bone1Name=dyn_hairBang_c_03, bone2Name=dyn_hairBang_c_02, swing1LimitDegrees=10, swing2LimitDegrees=30, twistLimitDegrees=5}
 {bone1Name=dyn_hairBang_c_04, bone2Name=dyn_hairBang_c_03, swing1LimitDegrees=10, swing2LimitDegrees=30, twistLimitDegrees=5}
 {bone1Name=dyn_hairBang_b_01, bone2Name=dyn_hairBang_b_MSTR, swing1LimitDegrees=10, swing2LimitDegrees=30, twistLimitDegrees=5}
 {bone1Name=dyn_hairBang_b_02, bone2Name=dyn_hairBang_b_01, swing1LimitDegrees=10, swing2LimitDegrees=30, twistLimitDegrees=5}
 {bone1Name=dyn_hairBang_b_03, bone2Name=dyn_hairBang_b_02, swing1LimitDegrees=10, swing2LimitDegrees=30, twistLimitDegrees=5}
 {bone1Name=dyn_hairBang_b_04, bone2Name=dyn_hairBang_b_03, swing1LimitDegrees=10, swing2LimitDegrees=30, twistLimitDegrees=5}
 {bone1Name=dyn_hairBang_a_01, bone2Name=dyn_hairBang_a_MSTR, swing1LimitDegrees=10, swing2LimitDegrees=30, twistLimitDegrees=5}
 {bone1Name=dyn_hairBang_a_02, bone2Name=dyn_hairBang_a_01, swing1LimitDegrees=10, swing2LimitDegrees=30, twistLimitDegrees=5}
 {bone1Name=dyn_hairBang_a_03, bone2Name=dyn_hairBang_a_02, swing1LimitDegrees=10, swing2LimitDegrees=30, twistLimitDegrees=5}
 {bone1Name=dyn_hairBang_a_04, bone2Name=dyn_hairBang_a_03, swing1LimitDegrees=10, swing2LimitDegrees=30, twistLimitDegrees=5}
]
